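# Editor utility list-entry widget for the Reference Replacement Tool, showing one affected asset with a thumbnail, name text and an X remove button.
# On list item assignment, it casts to ReferenceReplacementToolItem, stores it, and fills the text from the asset data's full name and sets the thumbnail.
# The remove button calls RemoveItself on the stored item.
# A RemoveCalled dispatcher is declared but unused in this graph.
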EditorUtilityWidgetBlueprint /FigmentTools/ReferenceReplacementTool/EUW_ReferenceReplacementToolItem
graphs [
 EventGraph {nodes [
   "On Clicked (button_remove)"() {is=title K2Node_ComponentBoundEvent}
   > RemoveItself(self:Reference Replacement Tool Item Object Reference=Get Item() -> (Item:Reference Replacement Tool Item Object Reference))
   n1 = Event OnListItemObjectSet()
   > n2 = Cast To ReferenceReplacementToolItem!(execute, Object=@n1.ListItemObject) -> (then, CastFailed, AsReference Replacement Tool Item:Reference Replacement Tool Item Object Reference) {is=title K2Node_DynamicCast}
   > Set Item(Item:Reference Replacement Tool Item Object Reference=@n2) -> (Output_Get:Reference Replacement Tool Item Object Reference)
   n3 = Get AffectedAsset!(self:Reference Replacement Tool Item Object Reference=Get Item()) -> (AffectedAsset:Asset Data Structure)
   > "SetText (Text)"(
    self=Get text_packageName() -> (text_packageName:Text Object Reference)
    InText="To Text (String)"(InString=GetFullName!(InAssetData:"Asset Data Structure (by ref)"=@n3) -> (ReturnValue:String))
   )
   > SetAsset!(
    execute
    self:Asset Thumbnail Widget Object Reference=Get AssetThumbnail() -> (AssetThumbnail:Asset Thumbnail Widget Object Reference)
    AssetData:"Asset Data Structure (by ref)"=Get AffectedAsset(self=Get Item())
   ) -> (then)
   Event PreConstruct()
   Event Construct()
   Event Tick()
  ]}
 RemoveCalled {nodes [RemoveCalled!() -> (then, Item:Reference Replacement Tool Item Object Reference) {is=title K2Node_FunctionEntry}]}
]
variables [
 Item {type=ReferenceReplacementToolItem, default=null}
 RemoveCalled {default=()}
]
widgets {
 info {rootWidgetClass=&HorizontalBox, widgetCount=6}
 widgets [
  HorizontalBox_61 {widgetClassPath=&HorizontalBox, widget=HorizontalBox_61}
  AssetThumbnail {
   widgetClassPath=/Script/UMGEditor.AssetThumbnailWidget
   bIsVariable=true
   parent=HorizontalBox_61
   widget=AssetThumbnail
   properties {thumbnailSettings {
     bAllowAssetSpecificThumbnailOverlay=false
     bAllowHintText=true
     bAllowRealTimeOnHovered=false
     bForceGenericThumbnail=false
     bOverrideAssetTypeColor=false
     colorStripOrientation=HorizontalBottomEdge
     genericThumbnailSize=64
     highlightedTextDelegate="(null).None"
     thumbnailLabel=AssetName
     assetTypeColorOverride=<rgba 0 0 0 0>
     hintColorAndOpacity=<rgba 0 0 0 0>
     padding=<ltrb 0 0 0 0>
    }}
   slot {class=HorizontalBoxSlot}
  }
  text_packageName {
   widgetClassPath=&TextBlock
   bIsVariable=true
   parent=HorizontalBox_61
   widget=text_packageName
   properties {
    text=Text Block
    font {
     bForceMonospaced=false
     bMaterialIsStencil=false
     fontMaterial=null
     fontObject=&NotoSans
     letterSpacing=0
     monospacedWidth=1
     size=8.25
     skewAmount=0
     typefaceFontName=NotoSans-Regular
     outlineSettings {
      bApplyOutlineToDropShadows=false
      bMiteredCorners=false
      bSeparateFillAlpha=false
      outlineMaterial=null
      outlineSize=0
      outlineColor=<rgba 0 0 0 1>
     }
    }
   }
   slot {class=HorizontalBoxSlot, properties {verticalAlignment=VAlign_Center, size {sizeRule=Fill, value=1}}}
  }
  button_remove {
   widgetClassPath=&EditorUtilityButton
   parent=HorizontalBox_61
   widget=button_remove
   slot {class=HorizontalBoxSlot, properties {padding=<ltrb 5 5 5 5>}}
  }
  X {
   widgetClassPath=&TextBlock
   parent=button_remove
   widget=X
   properties {
    text=X
    font {
     bForceMonospaced=false
     bMaterialIsStencil=false
     fontMaterial=null
     fontObject=&NotoSans
     letterSpacing=0
     monospacedWidth=1
     size=8.25
     skewAmount=0
     typefaceFontName=NotoSans-Regular
     outlineSettings {
      bApplyOutlineToDropShadows=false
      bMiteredCorners=false
      bSeparateFillAlpha=false
      outlineMaterial=null
      outlineSize=0
      outlineColor=<rgba 0 0 0 1>
     }
    }
   }
   slot {class=ButtonSlot, properties {horizontalAlignment=HAlign_Fill, padding=<ltrb 0 0 0 0>}}
  }
  Spacer_138 {widgetClassPath=&Spacer, parent=HorizontalBox_61, widget=Spacer_138, properties {size=<10 1>}, slot {class=HorizontalBoxSlot}}
 ]
}
