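# Editor utility widget for replacing asset references from a replacements data table, defaulting to DT_Figment_Replacements.
# PreConstruct creates a ReferenceReplacementToolModel, exposed through a property view; buttons add selected assets, clear them, or run ReplaceReferences.
# The affected-assets list filters by name, and the view shows a time estimate, or a whole-project warning when the list is empty.
# After replacement it shows a finished dialog, then closes its own tab using an ID built from its class path.
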EditorUtilityWidgetBlueprint /FigmentTools/ReferenceReplacementTool/ReferenceReplacementTool
graphs [
 CloseWidget {nodes [
   CloseWidget() {is=title K2Node_FunctionEntry}
   n1 = Self-Reference()
   n2 = Split(
    SourceString=GetSoftClassPath!(Class:Object Class Reference=Get Class(Object=@n1) -> (ReturnValue:Reference Replacement Tool Class Reference)) -> (ReturnValue_PathString:String)
    InStr=_C
    SearchCase=CaseSensitive
    SearchDir=FromEnd
   )
   > n3 = CloseTabByID!(
    execute
    self:Editor Utility Subsystem Object Reference=Get EditorUtilitySubsystem()
    NewTabID:Name=String To Name(InString=Append(A=@n2, B=_ActiveTab))
   ) -> (then, ReturnValue:bool)
   > Branch(Condition=@n3)
  ]}
 EventGraph {nodes [
   "On Clicked (ReplaceReferencesButton)"() {is=title K2Node_ComponentBoundEvent}
   > ReplaceReferences(self:Reference Replacement Tool Model Object Reference=Get ReferenceReplacementModel() -> (ReferenceReplacementModel:Reference Replacement Tool Model Object Reference))
   > ShowFinishedPopup()
   > CloseWidget() {is=title K2Node_CallFunction}
   "On Clicked (LoadAssetList)"() {is=title K2Node_ComponentBoundEvent}
   > AddSelectedAssets(self:Reference Replacement Tool Model Object Reference=Get ReferenceReplacementModel())
   > UpdateAffectedAssetsList()
   "On Clicked (ClearAssetList)"() {is=title K2Node_ComponentBoundEvent}
   > ClearAffectedAssets(self:Reference Replacement Tool Model Object Reference=Get ReferenceReplacementModel())
   > UpdateAffectedAssetsList()
   "On Text Committed (textbox_filter)"!() -> (OutputDelegate, then, Text:"Text (by ref)", CommitMethod:ETextCommit Enum=Default) {is=title K2Node_ComponentBoundEvent}
   > UpdateAffectedAssetsList()
   "On Clicked (button_ClearFilter)"() {is=title K2Node_ComponentBoundEvent}
   > "SetText (Text Box)"!(
    execute
    self:Text Box Object Reference=Get textbox_filter() -> (textbox_filter:Editor Utility Editable Text Box Object Reference)
    InText:Text
   ) -> (then)
   > UpdateAffectedAssetsList()
   Event PreConstruct()
   > n1 = Construct Object from Class(Class=/Script/FigmentTools.ReferenceReplacementToolModel, self=Self-Reference()) -> (ReturnValue:Reference Replacement Tool Model Object Reference)
   > n2 = Set ReferenceReplacementModel(ReferenceReplacementModel:Reference Replacement Tool Model Object Reference=@n1) -> (Output_Get:Reference Replacement Tool Model Object Reference)
   > SetObject(
    self=Get SingleProperty_ReplacementsDatatable() -> (SingleProperty_ReplacementsDatatable:Single Property View Object Reference)
    NewObject=@n2
   )
   > Set ReplacementsDataTable!(
    execute
    ReplacementsDataTable:Data Table Soft Object Reference=&/DT_Figment_Replacements
    self:Reference Replacement Tool Model Object Reference=Get ReferenceReplacementModel()
   ) -> (then, Output_Get:Data Table Soft Object Reference)
   > UpdateView()
   > Bind Event to On Item Removed(self:Reference Replacement Tool Model Object Reference=Get ReferenceReplacementModel(), Delegate=Create Event()) {is=title K2Node_AddDelegate}
   Event Construct()
   Event Tick()
   Reroute Node(InputPin:Wildcard) -> (OutputPin:Wildcard)
  ]}
 ShouldShowItem {nodes [
   n1 = ShouldShowItem!() -> (then, Item:Reference Replacement Tool Item Object Reference) {is=title K2Node_FunctionEntry}
   n2 = "GetText (Text Box)"!(self:Text Box Object Reference=Get textbox_filter()) -> (ReturnValue:Text)
   n3 = GetFullName!(InAssetData:"Asset Data Structure (by ref)"=Get AffectedAsset!(self:Reference Replacement Tool Item Object Reference=@n1) -> (AffectedAsset:Asset Data Structure)) -> (ReturnValue:String)
   n4 = "To String (Text)"(InText="GetText (Text Box)"(self=Get textbox_filter()))
   > Return Node!(execute, ReturnValue:bool=OR Boolean(A=TextIsEmpty(InText=@n2), B=Contains(SearchIn=@n3, Substring=@n4)))
  ]}
 ShowFinishedPopup {nodes [
   ShowFinishedPopup() {is=title K2Node_FunctionEntry}
   > Show Message Dialog(
    Title=MakeLiteralText(Value="LOCTEXT(\"Replacement Finished\")")
    Message=Format Text(Format="LOCTEXT(\"Reference Replacement Finished\")")
    MessageType=Ok
    DefaultValue=No
    MessageCategory=Success
   )
  ]}
 UpdateAffectedAssetsList {nodes [
   UpdateAffectedAssetsList() {is=title K2Node_FunctionEntry}
   > ClearListItems(self=Get ListView_AffectedAssets() -> (ListView_AffectedAssets:Editor Utility List View Object Reference))
   > n1 = For Each Loop(Array:Array of Reference Replacement Tool Item Object References=Get AffectedAssetItems!(self:Reference Replacement Tool Model Object Reference=Get ReferenceReplacementModel()) -> (AffectedAssetItems:Array of Reference Replacement Tool Item Object References)) -> (Array Element:Reference Replacement Tool Item Object Reference)
   UpdateView(execute=@n1.Completed)
   Branch(
    execute=@n1.LoopBody
    Condition=ShouldShowItem!(Item:Reference Replacement Tool Item Object Reference=@n1) -> (ReturnValue:bool) {is=title K2Node_CallFunction}
   )
   > AddItem!(execute, self:List View Object Reference=Get ListView_AffectedAssets(), Item:Object Reference=@n1) -> (then)
  ]}
 UpdateView {nodes [
   UpdateView() {is=title K2Node_FunctionEntry}
   n1 = Length(TargetArray:Array of Reference Replacement Tool Item Object References=Get AffectedAssetItems(self=Get ReferenceReplacementModel()))
   > SetVisibility(
    self=Get ZeroAffectedAssetsWarningBox() -> (ZeroAffectedAssetsWarningBox:Horizontal Box Object Reference)
    InVisibility=Select(Option 0=HitTestInvisible, Option 1=Collapsed, Index="integer > integer"(A=@n1))
   )
   n2 = Length(TargetArray=Get AffectedAssetItems(self=Get ReferenceReplacementModel()))
   > SetVisibility(
    self=Get txt_TimeEstimate() -> (txt_TimeEstimate:Text Object Reference)
    InVisibility=Select(Option 0=Collapsed, Option 1=HitTestInvisible, Index="integer > integer"(A=@n2))
   )
   n3 = GetTimeEstimateString!(self:Reference Replacement Tool Model Object Reference=Get ReferenceReplacementModel()) -> (ReturnValue:String)
   > "SetText (Text)"(
    self=Get txt_TimeEstimate()
    InText=Format Text!(Format="LOCTEXT(\"Estimated Time: {EstimatedTimeString}\")", EstimatedTimeString:Text="To Text (String)"(InString=@n3)) -> (Result)
   )
  ]}
]
variables [
 ReferenceReplacementModel {type=ReferenceReplacementToolModel, default=null}
 ReplaceReferencesDatatable {type=DataTable, default=&DT_Figment_Replacements}
]
widgets {
 info {rootWidgetClass=&VerticalBox, widgetCount=26}
 widgets [
  VerticalBox_159 {widgetClassPath=&VerticalBox, widget=VerticalBox_159}
  VerticalBox_161 {
   widgetClassPath=&VerticalBox
   parent=VerticalBox_159
   widget=VerticalBox_161
   slot {class=VerticalBoxSlot, properties {verticalAlignment=VAlign_Center, padding=<ltrb 4 2 4 2>}}
  }
  TextBlock_2 {
   widgetClassPath=&TextBlock
   parent=VerticalBox_161
   widget=TextBlock_2
   properties {
    text=Replacements
    font {
     bForceMonospaced=false
     bMaterialIsStencil=false
     fontMaterial=null
     fontObject=&Roboto
     letterSpacing=0
     monospacedWidth=1
     size=12
     skewAmount=0
     typefaceFontName=Bold
     outlineSettings {
      bApplyOutlineToDropShadows=false
      bMiteredCorners=false
      bSeparateFillAlpha=false
      outlineMaterial=null
      outlineSize=0
      outlineColor=<rgba 0 0 0 1>
     }
    }
   }
   slot {class=VerticalBoxSlot}
  }
  SingleProperty_ReplacementsDatatable {
   widgetClassPath=&SinglePropertyView
   parent=VerticalBox_161
   widget=SingleProperty_ReplacementsDatatable
   properties {nameOverride=Datatable, propertyName=ReplacementsDataTable}
   slot {class=VerticalBoxSlot}
  }
  VerticalBox_162 {
   widgetClassPath=&VerticalBox
   parent=VerticalBox_159
   widget=VerticalBox_162
   slot {class=VerticalBoxSlot, properties {verticalAlignment=VAlign_Top, padding=<ltrb 4 2 4 2>, size {sizeRule=Fill, value=1}}}
  }
  HorizontalBox_554 {
   widgetClassPath=&HorizontalBox
   parent=VerticalBox_162
   widget=HorizontalBox_554
   slot {class=VerticalBoxSlot, properties {padding=<ltrb 0 8 0 0>}}
  }
  TextBlock_1 {
   widgetClassPath=&TextBlock
   parent=HorizontalBox_554
   widget=TextBlock_1
   properties {
    text=Affected Assets
    font {
     bForceMonospaced=false
     bMaterialIsStencil=false
     fontMaterial=null
     fontObject=&Roboto
     letterSpacing=0
     monospacedWidth=1
     size=12
     skewAmount=0
     typefaceFontName=Bold
     outlineSettings {
      bApplyOutlineToDropShadows=false
      bMiteredCorners=false
      bSeparateFillAlpha=false
      outlineMaterial=null
      outlineSize=0
      outlineColor=<rgba 0 0 0 1>
     }
    }
   }
   slot {class=HorizontalBoxSlot, properties {verticalAlignment=VAlign_Center}}
  }
  LoadAssetList {
   widgetClassPath=&EditorUtilityButton
   parent=HorizontalBox_554
   widget=LoadAssetList
   properties {toolTipText=Add selected asets from content drawer}
   slot {class=HorizontalBoxSlot}
  }
  Image_332 {
   widgetClassPath=&Image
   bIsVariable=true
   parent=LoadAssetList
   widget=Image_332
   properties {
    toolTipText=Add selected asets from content drawer
    brush {
     bIsDynamicallyLoaded=false
     drawAs=Image
     imageType=FullColor
     mirroring=NoMirror
     resourceName=None
     resourceObject=/Game/Athena/UI/InGame/Creative/PropertyEditor/Icon_Picker/Textures_V2/64x64/T_UI_IconLibrary_LeftArrow_64
     tiling=NoTile
     imageSize=<16 16>
     margin=<ltrb 0 0 0 0>
     outlineSettings {
      bUseBrushTransparency=false
      roundingType=HalfHeightRadius
      width=0
      color {colorUseRule=UseColor_Specified, specifiedColor=<rgba 0 0 0 0>}
      cornerRadii=<0 0 0 0>
     }
     tintColor {colorUseRule=UseColor_Specified, specifiedColor=<rgba 1 1 1 1>}
     uVRegion {bIsValid=false, max=<0 0>, min=<0 0>}
    }
   }
   slot {class=ButtonSlot, properties {padding=<ltrb 0 0 0 0>}}
  }
  ClearAssetList {
   widgetClassPath=&EditorUtilityButton
   parent=HorizontalBox_554
   widget=ClearAssetList
   properties {toolTipText=Clear Affected Assets}
   slot {class=HorizontalBoxSlot}
  }
  Image {
   widgetClassPath=&Image
   bIsVariable=true
   parent=ClearAssetList
   widget=Image
   properties {
    toolTipText=Add selected asets from content drawer
    brush {
     bIsDynamicallyLoaded=false
     drawAs=Image
     imageType=FullColor
     mirroring=NoMirror
     resourceName=None
     resourceObject=/Game/Athena/UI/InGame/Creative/PropertyEditor/Icon_Picker/Textures_V2/64x64/T_UI_IconLibrary_TrashCan_64
     tiling=NoTile
     imageSize=<16 16>
     margin=<ltrb 0 0 0 0>
     outlineSettings {
      bUseBrushTransparency=false
      roundingType=HalfHeightRadius
      width=0
      color {colorUseRule=UseColor_Specified, specifiedColor=<rgba 0 0 0 0>}
      cornerRadii=<0 0 0 0>
     }
     tintColor {colorUseRule=UseColor_Specified, specifiedColor=<rgba 1 1 1 1>}
     uVRegion {bIsValid=false, max=<0 0>, min=<0 0>}
    }
   }
   slot {class=ButtonSlot, properties {padding=<ltrb 0 0 0 0>}}
  }
  HorizontalBox_503 {widgetClassPath=&HorizontalBox, parent=VerticalBox_162, widget=HorizontalBox_503, slot {class=VerticalBoxSlot}}
  textbox_filter {
   widgetClassPath=&EditorUtilityEditableTextBox
   bIsVariable=true
   parent=HorizontalBox_503
   widget=textbox_filter
   properties {
    hintText=Find
    widgetStyle {
     backgroundColor {colorUseRule=UseColor_Specified, specifiedColor=<rgba 1 1 1 1>}
     backgroundImageFocused {
      bIsDynamicallyLoaded=false
      drawAs=RoundedBox
      imageType=NoImage
      mirroring=NoMirror
      resourceName=None
      resourceObject=null
      tiling=NoTile
      imageSize=<0 0>
      margin=<ltrb 0 0 0 0>
      outlineSettings {
       bUseBrushTransparency=false
       roundingType=FixedRadius
       width=1
       color {colorUseRule=UseColor_ColorTable, specifiedColor=<rgba 0 0 0 0>}
       cornerRadii=<4 4 4 4>
      }
      tintColor {colorUseRule=UseColor_Specified, specifiedColor=<rgba 0.0048 0.0048 0.0048 1>}
      uVRegion {bIsValid=false, max=<0 0>, min=<0 0>}
     }
     backgroundImageHovered {
      bIsDynamicallyLoaded=false
      drawAs=RoundedBox
      imageType=NoImage
      mirroring=NoMirror
      resourceName=None
      resourceObject=null
      tiling=NoTile
      imageSize=<0 0>
      margin=<ltrb 0 0 0 0>
      outlineSettings {
       bUseBrushTransparency=false
       roundingType=FixedRadius
       width=1
       color {colorUseRule=UseColor_ColorTable, specifiedColor=<rgba 0 0 0 0>}
       cornerRadii=<4 4 4 4>
      }
      tintColor {colorUseRule=UseColor_Specified, specifiedColor=<rgba 0.0048 0.0048 0.0048 1>}
      uVRegion {bIsValid=false, max=<0 0>, min=<0 0>}
     }
     backgroundImageNormal {
      bIsDynamicallyLoaded=false
      drawAs=RoundedBox
      imageType=NoImage
      mirroring=NoMirror
      resourceName=None
      resourceObject=null
      tiling=NoTile
      imageSize=<0 0>
      margin=<ltrb 0 0 0 0>
      outlineSettings {
       bUseBrushTransparency=false
       roundingType=FixedRadius
       width=1
       color {colorUseRule=UseColor_ColorTable, specifiedColor=<rgba 0 0 0 0>}
       cornerRadii=<4 4 4 4>
      }
      tintColor {colorUseRule=UseColor_Specified, specifiedColor=<rgba 0.0048 0.0048 0.0048 1>}
      uVRegion {bIsValid=false, max=<0 0>, min=<0 0>}
     }
     backgroundImageReadOnly {
      bIsDynamicallyLoaded=false
      drawAs=RoundedBox
      imageType=NoImage
      mirroring=NoMirror
      resourceName=None
      resourceObject=null
      tiling=NoTile
      imageSize=<0 0>
      margin=<ltrb 0 0 0 0>
      outlineSettings {
       bUseBrushTransparency=false
       roundingType=FixedRadius
       width=1
       color {colorUseRule=UseColor_ColorTable, specifiedColor=<rgba 0 0 0 0>}
       cornerRadii=<4 4 4 4>
      }
      tintColor {colorUseRule=UseColor_Specified, specifiedColor=<rgba 0.0284 0.0284 0.0284 1>}
      uVRegion {bIsValid=false, max=<0 0>, min=<0 0>}
     }
     focusedForegroundColor {colorUseRule=UseColor_Specified, specifiedColor=<rgba 1 1 1 1>}
     foregroundColor {colorUseRule=UseColor_Specified, specifiedColor=<rgba 0.5271 0.5271 0.5271 1>}
     hScrollBarPadding=<ltrb 0 0 0 0>
     padding=<ltrb 0 0 0 0>
     readOnlyForegroundColor {colorUseRule=UseColor_Foreground, specifiedColor=<rgba 1 0 1 1>}
     scrollBarStyle {
      thickness=8
      draggedThumbImage {
       bIsDynamicallyLoaded=false
       drawAs=RoundedBox
       imageType=NoImage
       mirroring=NoMirror
       resourceName=None
       resourceObject=null
       tiling=NoTile
       imageSize=<0 0>
       margin=<ltrb 0 0 0 0>
       outlineSettings {
        bUseBrushTransparency=false
        roundingType=FixedRadius
        width=0
        color {colorUseRule=UseColor_Specified, specifiedColor=<rgba 0 0 0 0>}
        cornerRadii=<4 4 4 4>
       }
       tintColor {colorUseRule=UseColor_Specified, specifiedColor=<rgba 0.2159 0.2159 0.2159 1>}
       uVRegion {bIsValid=false, max=<0 0>, min=<0 0>}
      }
      horizontalBackgroundImage {
       bIsDynamicallyLoaded=false
       drawAs=NoDrawType
       imageType=NoImage
       mirroring=NoMirror
       resourceName=None
       resourceObject=null
       tiling=NoTile
       imageSize=<0 0>
       margin=<ltrb 0 0 0 0>
       outlineSettings {
        bUseBrushTransparency=false
        roundingType=HalfHeightRadius
        width=0
        color {colorUseRule=UseColor_Specified, specifiedColor=<rgba 0 0 0 0>}
        cornerRadii=<0 0 0 0>
       }
       tintColor {colorUseRule=UseColor_Specified, specifiedColor=<rgba 1 1 1 1>}
       uVRegion {bIsValid=false, max=<0 0>, min=<0 0>}
      }
      horizontalBottomSlotImage {
       bIsDynamicallyLoaded=false
       drawAs=NoDrawType
       imageType=NoImage
       mirroring=NoMirror
       resourceName=None
       resourceObject=null
       tiling=NoTile
       imageSize=<0 0>
       margin=<ltrb 0 0 0 0>
       outlineSettings {
        bUseBrushTransparency=false
        roundingType=HalfHeightRadius
        width=0
        color {colorUseRule=UseColor_Specified, specifiedColor=<rgba 0 0 0 0>}
        cornerRadii=<0 0 0 0>
       }
       tintColor {colorUseRule=UseColor_Specified, specifiedColor=<rgba 1 1 1 1>}
       uVRegion {bIsValid=false, max=<0 0>, min=<0 0>}
      }
      horizontalTopSlotImage {
       bIsDynamicallyLoaded=false
       drawAs=NoDrawType
       imageType=NoImage
       mirroring=NoMirror
       resourceName=None
       resourceObject=null
       tiling=NoTile
       imageSize=<0 0>
       margin=<ltrb 0 0 0 0>
       outlineSettings {
        bUseBrushTransparency=false
        roundingType=HalfHeightRadius
        width=0
        color {colorUseRule=UseColor_Specified, specifiedColor=<rgba 0 0 0 0>}
        cornerRadii=<0 0 0 0>
       }
       tintColor {colorUseRule=UseColor_Specified, specifiedColor=<rgba 1 1 1 1>}
       uVRegion {bIsValid=false, max=<0 0>, min=<0 0>}
      }
      hoveredThumbImage {
       bIsDynamicallyLoaded=false
       drawAs=RoundedBox
       imageType=NoImage
       mirroring=NoMirror
       resourceName=None
       resourceObject=null
       tiling=NoTile
       imageSize=<0 0>
       margin=<ltrb 0 0 0 0>
       outlineSettings {
        bUseBrushTransparency=false
        roundingType=FixedRadius
        width=0
        color {colorUseRule=UseColor_Specified, specifiedColor=<rgba 0 0 0 0>}
        cornerRadii=<4 4 4 4>
       }
       tintColor {colorUseRule=UseColor_Specified, specifiedColor=<rgba 0.2159 0.2159 0.2159 1>}
       uVRegion {bIsValid=false, max=<0 0>, min=<0 0>}
      }
      normalThumbImage {
       bIsDynamicallyLoaded=false
       drawAs=RoundedBox
       imageType=NoImage
       mirroring=NoMirror
       resourceName=None
       resourceObject=null
       tiling=NoTile
       imageSize=<0 0>
       margin=<ltrb 0 0 0 0>
       outlineSettings {
        bUseBrushTransparency=false
        roundingType=FixedRadius
        width=0
        color {colorUseRule=UseColor_Specified, specifiedColor=<rgba 0 0 0 0>}
        cornerRadii=<4 4 4 4>
       }
       tintColor {colorUseRule=UseColor_Specified, specifiedColor=<rgba 0.0953 0.0953 0.0953 1>}
       uVRegion {bIsValid=false, max=<0 0>, min=<0 0>}
      }
      verticalBackgroundImage {
       bIsDynamicallyLoaded=false
       drawAs=NoDrawType
       imageType=NoImage
       mirroring=NoMirror
       resourceName=None
       resourceObject=null
       tiling=NoTile
       imageSize=<0 0>
       margin=<ltrb 0 0 0 0>
       outlineSettings {
        bUseBrushTransparency=false
        roundingType=HalfHeightRadius
        width=0
        color {colorUseRule=UseColor_Specified, specifiedColor=<rgba 0 0 0 0>}
        cornerRadii=<0 0 0 0>
       }
       tintColor {colorUseRule=UseColor_Specified, specifiedColor=<rgba 1 1 1 1>}
       uVRegion {bIsValid=false, max=<0 0>, min=<0 0>}
      }
      verticalBottomSlotImage {
       bIsDynamicallyLoaded=false
       drawAs=NoDrawType
       imageType=NoImage
       mirroring=NoMirror
       resourceName=None
       resourceObject=null
       tiling=NoTile
       imageSize=<0 0>
       margin=<ltrb 0 0 0 0>
       outlineSettings {
        bUseBrushTransparency=false
        roundingType=HalfHeightRadius
        width=0
        color {colorUseRule=UseColor_Specified, specifiedColor=<rgba 0 0 0 0>}
        cornerRadii=<0 0 0 0>
       }
       tintColor {colorUseRule=UseColor_Specified, specifiedColor=<rgba 1 1 1 1>}
       uVRegion {bIsValid=false, max=<0 0>, min=<0 0>}
      }
      verticalTopSlotImage {
       bIsDynamicallyLoaded=false
       drawAs=NoDrawType
       imageType=NoImage
       mirroring=NoMirror
       resourceName=None
       resourceObject=null
       tiling=NoTile
       imageSize=<0 0>
       margin=<ltrb 0 0 0 0>
       outlineSettings {
        bUseBrushTransparency=false
        roundingType=HalfHeightRadius
        width=0
        color {colorUseRule=UseColor_Specified, specifiedColor=<rgba 0 0 0 0>}
        cornerRadii=<0 0 0 0>
       }
       tintColor {colorUseRule=UseColor_Specified, specifiedColor=<rgba 1 1 1 1>}
       uVRegion {bIsValid=false, max=<0 0>, min=<0 0>}
      }
     }
     textStyle {
      overflowPolicy=Clip
      transformPolicy=None
      colorAndOpacity {colorUseRule=UseColor_Foreground, specifiedColor=<rgba 1 0 1 1>}
      font {
       bForceMonospaced=false
       bMaterialIsStencil=false
       fontMaterial=null
       fontObject=/LauncherSocial/BrutalType
       letterSpacing=0
       monospacedWidth=1
       size=10
       skewAmount=0
       typefaceFontName=Regular
       outlineSettings {
        bApplyOutlineToDropShadows=false
        bMiteredCorners=false
        bSeparateFillAlpha=false
        outlineMaterial=null
        outlineSize=0
        outlineColor=<rgba 0 0 0 1>
       }
      }
      highlightColor {colorUseRule=UseColor_ColorTable, specifiedColor=<rgba 0 0 0 0>}
      highlightShape {
       bIsDynamicallyLoaded=false
       drawAs=Image
       imageType=NoImage
       mirroring=NoMirror
       resourceName=None
       resourceObject=null
       tiling=NoTile
       imageSize=<0 0>
       margin=<ltrb 0 0 0 0>
       outlineSettings {
        bUseBrushTransparency=false
        roundingType=HalfHeightRadius
        width=0
        color {colorUseRule=UseColor_Specified, specifiedColor=<rgba 0 0 0 0>}
        cornerRadii=<0 0 0 0>
       }
       tintColor {colorUseRule=UseColor_ColorTable, specifiedColor=<rgba 0 0 0 0>}
       uVRegion {bIsValid=false, max=<0 0>, min=<0 0>}
      }
      selectedBackgroundColor {colorUseRule=UseColor_ColorTable, specifiedColor=<rgba 0 0 0 0>}
      shadowColorAndOpacity=<rgba 0 0 0 1>
      shadowOffset=<0 0>
      strikeBrush {
       bIsDynamicallyLoaded=false
       drawAs=Image
       imageType=NoImage
       mirroring=NoMirror
       resourceName=None
       resourceObject=null
       tiling=NoTile
       imageSize=<32 32>
       margin=<ltrb 0 0 0 0>
       outlineSettings {
        bUseBrushTransparency=false
        roundingType=HalfHeightRadius
        width=0
        color {colorUseRule=UseColor_Specified, specifiedColor=<rgba 0 0 0 0>}
        cornerRadii=<0 0 0 0>
       }
       tintColor {colorUseRule=UseColor_Specified, specifiedColor=<rgba 1 1 1 1>}
       uVRegion {bIsValid=false, max=<0 0>, min=<0 0>}
      }
      underlineBrush {
       bIsDynamicallyLoaded=false
       drawAs=Image
       imageType=NoImage
       mirroring=NoMirror
       resourceName=None
       resourceObject=null
       tiling=NoTile
       imageSize=<32 32>
       margin=<ltrb 0 0 0 0>
       outlineSettings {
        bUseBrushTransparency=false
        roundingType=HalfHeightRadius
        width=0
        color {colorUseRule=UseColor_Specified, specifiedColor=<rgba 0 0 0 0>}
        cornerRadii=<0 0 0 0>
       }
       tintColor {colorUseRule=UseColor_Specified, specifiedColor=<rgba 1 1 1 1>}
       uVRegion {bIsValid=false, max=<0 0>, min=<0 0>}
      }
     }
     vScrollBarPadding=<ltrb 0 0 0 0>
    }
   }
   slot {class=HorizontalBoxSlot, properties {verticalAlignment=VAlign_Center, size {sizeRule=Fill, value=1}}}
  }
  button_ClearFilter {
   widgetClassPath=&EditorUtilityButton
   parent=HorizontalBox_503
   widget=button_ClearFilter
   slot {class=HorizontalBoxSlot, properties {horizontalAlignment=HAlign_Left, verticalAlignment=VAlign_Center}}
  }
  TextBlock_410 {
   widgetClassPath=&TextBlock
   parent=button_ClearFilter
   widget=TextBlock_410
   properties {
    text=x
    font {
     bForceMonospaced=false
     bMaterialIsStencil=false
     fontMaterial=null
     fontObject=&Roboto
     letterSpacing=0
     monospacedWidth=1
     size=10.5
     skewAmount=0
     typefaceFontName=Bold
     outlineSettings {
      bApplyOutlineToDropShadows=false
      bMiteredCorners=false
      bSeparateFillAlpha=false
      outlineMaterial=null
      outlineSize=0
      outlineColor=<rgba 0 0 0 1>
     }
    }
   }
   slot {class=ButtonSlot}
  }
  EditorUtilityScrollBox_668 {
   widgetClassPath=&EditorUtilityScrollBox
   parent=VerticalBox_162
   widget=EditorUtilityScrollBox_668
   properties {alwaysShowScrollbar=true}
   slot {class=VerticalBoxSlot, properties {size {sizeRule=Fill, value=1}}}
  }
  ListView_AffectedAssets {
   widgetClassPath=/Script/Blutility.EditorUtilityListView
   bIsVariable=true
   parent=EditorUtilityScrollBox_668
   widget=ListView_AffectedAssets
   properties {entryWidgetClass=&/EUW_ReferenceReplacementToolItem.EUW_ReferenceReplacementToolItem_C}
   slot {class=ScrollBoxSlot}
  }
  ZeroAffectedAssetsWarningBox {
   widgetClassPath=&HorizontalBox
   bIsVariable=true
   parent=VerticalBox_159
   widget=ZeroAffectedAssetsWarningBox
   slot {class=VerticalBoxSlot, properties {verticalAlignment=VAlign_Bottom, padding=<ltrb 4 4 4 4>}}
  }
  Image_138 {
   widgetClassPath=&Image
   bIsVariable=true
   parent=ZeroAffectedAssetsWarningBox
   widget=Image_138
   properties {brush {
     bIsDynamicallyLoaded=false
     drawAs=Image
     imageType=FullColor
     mirroring=NoMirror
     resourceName=None
     resourceObject=null
     tiling=NoTile
     imageSize=<32 32>
     margin=<ltrb 0 0 0 0>
     outlineSettings {
      bUseBrushTransparency=false
      roundingType=HalfHeightRadius
      width=0
      color {colorUseRule=UseColor_Specified, specifiedColor=<rgba 0 0 0 0>}
      cornerRadii=<0 0 0 0>
     }
     tintColor {colorUseRule=UseColor_Specified, specifiedColor=<rgba 1 0.8765 0 1>}
     uVRegion {bIsValid=false, max=<0 0>, min=<0 0>}
    }}
   slot {class=HorizontalBoxSlot, properties {verticalAlignment=VAlign_Center}}
  }
  Border_0 {
   widgetClassPath=&Border
   parent=ZeroAffectedAssetsWarningBox
   widget=Border_0
   properties {
    background {
     bIsDynamicallyLoaded=false
     drawAs=Box
     imageType=NoImage
     mirroring=NoMirror
     resourceName=None
     resourceObject=null
     tiling=NoTile
     imageSize=<32 32>
     margin=<ltrb 0 0 1 1>
     outlineSettings {
      bUseBrushTransparency=false
      roundingType=HalfHeightRadius
      width=-0.1447
      color {colorUseRule=UseColor_Specified, specifiedColor=<rgba 0 0 0 0>}
      cornerRadii=<0 0 0 0>
     }
     tintColor {colorUseRule=UseColor_Specified, specifiedColor=<rgba 0.2104 0.1848 0 1>}
     uVRegion {bIsValid=false, max=<0 0>, min=<0 0>}
    }
    brushColor=<rgba 1 0.9482 0 1>
    contentColorAndOpacity=<rgba 0.8443 0.8424 1 1>
    padding=<ltrb 4 4 4 4>
   }
   slot {class=HorizontalBoxSlot, properties {size {sizeRule=Fill, value=1}}}
  }
  TextBlock_0 {
   widgetClassPath=&TextBlock
   parent=Border_0
   widget=TextBlock_0
   properties {
    justification=Center
    text="0 Affected Assets Selected. \r\nReplace references will affect the whole project."
    colorAndOpacity {colorUseRule=UseColor_Specified, specifiedColor=<rgba 1 0.992 0.9512 1>}
    font {
     bForceMonospaced=false
     bMaterialIsStencil=false
     fontMaterial=null
     fontObject=&Roboto
     letterSpacing=0
     monospacedWidth=1
     size=9
     skewAmount=0
     typefaceFontName=Bold
     outlineSettings {
      bApplyOutlineToDropShadows=false
      bMiteredCorners=false
      bSeparateFillAlpha=false
      outlineMaterial=null
      outlineSize=1
      outlineColor=<rgba 0 0 0 1>
     }
    }
   }
   slot {class=BorderSlot, properties {padding=<ltrb 4 4 4 4>}}
  }
  txt_TimeEstimate {
   widgetClassPath=&TextBlock
   bIsVariable=true
   parent=VerticalBox_159
   widget=txt_TimeEstimate
   properties {
    justification=Center
    text="Estimated Time: "
    colorAndOpacity {colorUseRule=UseColor_Specified, specifiedColor=<rgba 1 0.992 0.9512 1>}
    font {
     bForceMonospaced=false
     bMaterialIsStencil=false
     fontMaterial=null
     fontObject=&Roboto
     letterSpacing=0
     monospacedWidth=1
     size=9
     skewAmount=0
     typefaceFontName=Bold
     outlineSettings {
      bApplyOutlineToDropShadows=false
      bMiteredCorners=false
      bSeparateFillAlpha=false
      outlineMaterial=null
      outlineSize=1
      outlineColor=<rgba 0 0 0 1>
     }
    }
   }
   slot {class=VerticalBoxSlot, properties {horizontalAlignment=HAlign_Left, padding=<ltrb 4 4 4 4>}}
  }
  HorizontalBox_429 {
   widgetClassPath=&HorizontalBox
   parent=VerticalBox_159
   widget=HorizontalBox_429
   slot {class=VerticalBoxSlot, properties {verticalAlignment=VAlign_Bottom}}
  }
  ReplaceReferencesButton {
   widgetClassPath=&EditorUtilityButton
   parent=HorizontalBox_429
   widget=ReplaceReferencesButton
   slot {class=HorizontalBoxSlot, properties {padding=<ltrb 8 8 8 8>, size {sizeRule=Fill, value=1}}}
  }
  VerticalBox_0 {widgetClassPath=&VerticalBox, parent=ReplaceReferencesButton, widget=VerticalBox_0, slot {class=ButtonSlot}}
  TextBlock_54 {
   widgetClassPath=&TextBlock
   parent=VerticalBox_0
   widget=TextBlock_54
   properties {
    text=Replace References
    font {
     bForceMonospaced=false
     bMaterialIsStencil=false
     fontMaterial=null
     fontObject=&Roboto
     letterSpacing=0
     monospacedWidth=1
     size=13.5
     skewAmount=0
     typefaceFontName=Bold
     outlineSettings {
      bApplyOutlineToDropShadows=false
      bMiteredCorners=false
      bSeparateFillAlpha=false
      outlineMaterial=null
      outlineSize=0
      outlineColor=<rgba 0 0 0 1>
     }
    }
   }
   slot {class=VerticalBoxSlot}
  }
 ]
}
